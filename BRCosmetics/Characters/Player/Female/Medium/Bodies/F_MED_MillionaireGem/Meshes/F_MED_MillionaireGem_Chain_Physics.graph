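# Physics asset for the chain and amulet of the female medium MillionaireGem character body, giving the necklace physical bodies to swing on the upper spine.
# Four capsule bodies link in series from spine_05 through base_chain_m and dyn_chain_m to dyn_amulet_m, with the large spine capsule standing in for the torso.
# The base-to-spine joint is angularly free, while the chain and amulet joints have asymmetric swing limits, with twist allowance only at the amulet.
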PhysicsAsset /BRCosmetics/Characters/Player/Female/Medium/Bodies/F_MED_MillionaireGem/Meshes/F_MED_MillionaireGem_Chain_Physics
bodies {
 base_chain_m [{shapeType=Capsule, length=0.505, radius=0.505, shapeName=base_chain_m_capsule, rotation=<rot 0 -90 0>}]
 dyn_amulet_m [{shapeType=Capsule, length=2.5, radius=2, shapeName=dyn_amulet_m_capsule, center=<2.8 0 0>, rotation=<rot 90 0 0>}]
 dyn_chain_m [{shapeType=Capsule, length=5, radius=2, shapeName=dyn_chain_m_capsule, center=<11.6454 0 -0.2272>, rotation=<rot 90 0 0>}]
 spine_05 [{shapeType=Capsule, length=13.3401, radius=7.692, shapeName=spine_05_capsule, center=<-9.7296 3.9075 0>, rotation=<rot 0 0 180>}]
}
constraints [
 {bone1Name=dyn_chain_m, bone2Name=base_chain_m, swing1LimitDegrees=5, swing2LimitDegrees=50, twistLimitDegrees=0}
 {bone1Name=dyn_amulet_m, bone2Name=dyn_chain_m, swing1LimitDegrees=5, swing2LimitDegrees=30, twistLimitDegrees=45}
 {
  bone1Name=base_chain_m
  bone2Name=spine_05
  swing1LimitDegrees=45
  swing1Motion=Free
  swing2LimitDegrees=45
  swing2Motion=Free
  twistLimitDegrees=45
  twistMotion=Free
 }
]
