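# Animation Blueprint for the female MillionaireGem player body, adding secondary motion to collar, jacket, chain, gun and knife.
# It copies the attached parent's pose through two Control Rigs, applies additive collar and jacket bone fixes, then runs AnimDynamics and RigidBody simulation.
# The event graph updates gravity via a deprecated function and sets fix alphas from skydive, parachute, crouch, look-up and melee head-turn states.
# isRunning drives fixes but is never set here.
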AnimBlueprint /BRCosmetics/Characters/Player/Female/Medium/Bodies/F_MED_MillionaireGem/Meshes/F_MED_MillionaireGem_AnimBP
target_skeleton=&/F_MED_MillionaireGem_Skeleton
graphs [
 AnimGraph {nodes [
   Output Pose(Result=Use cached pose 'CRCache_Dyn'()) {layerGroup=None}
   n1 = Control Rig(Source=Copy Pose From Mesh() {bUseAttachedParent=true}) {
    alphaBoolBlend {blendOption=Linear}
    controlRigAssetReference {blueprintRigClass=&/F_MED_MillionaireGem_DeformRig_Master_ControlRig_new.F_MED_MillionaireGem_DeformRig_Master_ControlRig_new_C}
   }
   n2 = Control Rig(Source=@n1) {
    alphaBoolBlend {blendOption=Linear}
    controlRigAssetReference {blueprintRigClass=&/F_MED_MillionaireGem_CtrlRig.F_MED_MillionaireGem_CtrlRig_C}
   }
   Save cached pose 'CRCache_ControlRig'(Pose=@n2) {cacheName=CRCache_ControlRig}
   n3 = Anim Dynamics - Bone: dyn_gun_r(ComponentPose=Local To Component(LocalPose=Use cached pose 'CRCache_JacketFix'()), Alpha=1, ExternalForce="0,0,0") {
    angularDampingOverride=1
    angularSpringConstant=10000
    bAngularSpring=true
    bOverrideAngularDamping=true
    bUseGravityOverride=true
    gravityScale=0.35
    simulationSpace=RootRelative
    alphaBoolBlend {blendOption=Linear}
    boundBone {boneName=dyn_gun_r}
    componentLinearAccScale=<0.5 0.5 0.5>
    componentLinearVelScale=<0.5 0.5 0.5>
    gravityOverride {z=-980}
    physicsBodyDefinitions [{
      boundBone {boneName=dyn_gun_r}
      boxExtents=<25 35 35>
      constraintSetup {angularLimitsMax=<0 1 3>, angularLimitsMin=<0 -1 -3>, angularTarget=<1 0 0>}
      localJointOffset=<1 0 0>
     }]
   }
   n4 = Anim Dynamics - Bone: dyn_knife_r(ComponentPose=@n3, Alpha=1, ExternalForce="0,0,0") {
    angularDampingOverride=1
    angularSpringConstant=10000
    bAngularSpring=true
    bOverrideAngularDamping=true
    bUseGravityOverride=true
    gravityScale=0.35
    simulationSpace=RootRelative
    alphaBoolBlend {blendOption=Linear}
    boundBone {boneName=dyn_knife_r}
    componentLinearAccScale=<0.8 0.8 0.8>
    componentLinearVelScale=<0.8 0.8 0.8>
    gravityOverride {z=-980}
    physicsBodyDefinitions [{
      boundBone {boneName=dyn_knife_r}
      boxExtents=<25 35 35>
      constraintSetup {angularLimitsMax=<0 2 7>, angularLimitsMin=<0 -2 -7>, angularTarget=<1 0 0>}
      localJointOffset=<1 0 0>
     }]
   }
   n5 = Get GravityOverride()
   n6 = RigidBody!(ComponentPose=@n4, Alpha, OverrideWorldGravity:Vector=@n5, ExternalForce) -> (Pose) {
    bOverrideWorldGravity=true
    overridePhysicsAsset=&/F_MED_MillionaireGem_Jacket_Physics
    simulationSpace=BaseBoneSpace
    alphaBoolBlend {blendOption=Linear}
    baseBoneRef {boneName=spine_02}
    componentAppliedLinearAccClamp=<3000 3000 3000>
    componentLinearAccScale=<0.5 0.5 0.5>
    componentLinearVelScale=<0.5 0.5 0.5>
    simSpaceSettings {
     maxAngularAcceleration=1000
     maxAngularVelocity=3
     maxLinearAcceleration=700
     maxLinearVelocity=500
     velocityScaleZ=0
     worldAlpha=0.3
     externalLinearDragV=<0.2 0.2>
    }
   }
   n7 = RigidBody(ComponentPose=@n6, OverrideWorldGravity=@n5) {
    bClampLinearTranslationLimitToRefPose=true
    bOverrideWorldGravity=true
    overridePhysicsAsset=&/F_MED_MillionaireGem_Chain_Physics
    simulationSpace=BaseBoneSpace
    alphaBoolBlend {blendOption=Linear}
    baseBoneRef {boneName=spine_04}
    componentAppliedLinearAccClamp=<3000 3000 3000>
    componentLinearAccScale=<0.5 0.5 0.5>
    componentLinearVelScale=<0.5 0.5 0.5>
    overrideWorldGravity {z=-980}
    simSpaceSettings {
     maxAngularAcceleration=1000
     maxAngularVelocity=3
     maxLinearAcceleration=700
     maxLinearVelocity=500
     velocityScaleZ=0
     worldAlpha=0.5
     externalLinearDragV=<1 1>
    }
   }
   n8 = Anim Dynamics - Bone: dyn_collar_bk_l!(ComponentPose=@n7, Alpha=1, GravityOverride:Vector="0,0,-980", ExternalForce="0,0,0") -> (Pose) {
    angularDampingOverride=1
    angularSpringConstant=10000
    bAngularSpring=true
    bOverrideAngularDamping=true
    bUseGravityOverride=true
    bUsePlanarLimit=false
    gravityScale=0.35
    simulationSpace=BoneRelative
    alphaBoolBlend {blendOption=Linear}
    boundBone {boneName=dyn_collar_bk_l}
    componentAppliedLinearAccClamp=<3000 3000 3000>
    componentLinearAccScale=<0.2 0.2 0.2>
    componentLinearVelScale=<0.2 0.2 0.2>
    gravityOverride {z=-980}
    physicsBodyDefinitions [{
      boundBone {boneName=dyn_collar_bk_l}
      boxExtents=<30 40 40>
      constraintSetup {angularLimitsMax=<4 4 4>, angularLimitsMin=<-4 -4 -4>, angularTarget=<1 0 0>}
      localJointOffset=<0.5 0 0>
     }]
    relativeSpaceBone {boneName=spine_03}
   }
   n9 = Anim Dynamics - Bone: dyn_collar_fr_r!(ComponentPose=@n8, Alpha=1, GravityOverride:Vector="0,0,-980", ExternalForce="0,0,0") -> (Pose) {
    angularDampingOverride=1
    angularSpringConstant=10000
    bAngularSpring=true
    bOverrideAngularDamping=true
    bUseGravityOverride=true
    gravityScale=0.35
    simulationSpace=BoneRelative
    alphaBoolBlend {blendOption=Linear}
    boundBone {boneName=dyn_collar_fr_r}
    componentAppliedLinearAccClamp=<3000 3000 3000>
    componentLinearAccScale=<0.2 0.2 0.2>
    componentLinearVelScale=<0.2 0.2 0.2>
    gravityOverride {z=-980}
    physicsBodyDefinitions [{
      boundBone {boneName=dyn_collar_fr_r}
      boxExtents=<30 40 40>
      constraintSetup {angularLimitsMax=<4 4 4>, angularLimitsMin=<-4 -4 -4>, angularTarget=<1 0 0>}
      localJointOffset=<0.5 0 0>
     }]
    relativeSpaceBone {boneName=spine_03}
   }
   n10 = Anim Dynamics - Bone: dyn_collar_fr_l!(ComponentPose=@n9, Alpha=1, GravityOverride:Vector="0,0,-980", ExternalForce="0,0,0") -> (Pose) {
    angularDampingOverride=1
    angularSpringConstant=10000
    bAngularSpring=true
    bOverrideAngularDamping=true
    bUseGravityOverride=true
    gravityScale=0.35
    simulationSpace=BoneRelative
    alphaBoolBlend {blendOption=Linear}
    boundBone {boneName=dyn_collar_fr_l}
    componentAppliedLinearAccClamp=<3000 3000 3000>
    componentLinearAccScale=<0.2 0.2 0.2>
    componentLinearVelScale=<0.2 0.2 0.2>
    gravityOverride {z=-980}
    physicsBodyDefinitions [{
      boundBone {boneName=dyn_collar_fr_l}
      boxExtents=<30 40 40>
      constraintSetup {angularLimitsMax=<4 4 4>, angularLimitsMin=<-4 -4 -4>, angularTarget=<1 0 0>}
      localJointOffset=<0.5 0 0>
     }]
    relativeSpaceBone {boneName=spine_03}
   }
   n11 = Anim Dynamics - Bone: dyn_collar_side_l!(ComponentPose=@n10, Alpha=1, GravityOverride:Vector="0,0,-980", ExternalForce="0,0,0") -> (Pose) {
    angularDampingOverride=1
    angularSpringConstant=10000
    bAngularSpring=true
    bOverrideAngularDamping=true
    bUseGravityOverride=true
    gravityScale=0.35
    simulationSpace=BoneRelative
    alphaBoolBlend {blendOption=Linear}
    boundBone {boneName=dyn_collar_side_l}
    componentAppliedLinearAccClamp=<3000 3000 3000>
    componentLinearAccScale=<0.2 0.2 0.2>
    componentLinearVelScale=<0.2 0.2 0.2>
    gravityOverride {z=-980}
    physicsBodyDefinitions [{
      boundBone {boneName=dyn_collar_side_l}
      boxExtents=<30 40 40>
      constraintSetup {angularLimitsMax=<4 4 4>, angularLimitsMin=<-4 -4 -4>, angularTarget=<1 0 0>}
      localJointOffset=<0.5 0 0>
     }]
    relativeSpaceBone {boneName=spine_03}
   }
   n12 = Anim Dynamics - Bone: dyn_collar_side_r!(ComponentPose=@n11, Alpha=1, GravityOverride:Vector="0,0,-980", ExternalForce="0,0,0") -> (Pose) {
    angularDampingOverride=1
    angularSpringConstant=10000
    bAngularSpring=true
    bOverrideAngularDamping=true
    bUseGravityOverride=true
    gravityScale=0.35
    simulationSpace=BoneRelative
    alphaBoolBlend {blendOption=Linear}
    boundBone {boneName=dyn_collar_side_r}
    componentAppliedLinearAccClamp=<3000 3000 3000>
    componentLinearAccScale=<0.2 0.2 0.2>
    componentLinearVelScale=<0.2 0.2 0.2>
    gravityOverride {z=-980}
    physicsBodyDefinitions [{
      boundBone {boneName=dyn_collar_side_r}
      boxExtents=<30 40 40>
      constraintSetup {angularLimitsMax=<4 4 4>, angularLimitsMin=<-4 -4 -4>, angularTarget=<1 0 0>}
      localJointOffset=<0.5 0 0>
     }]
    relativeSpaceBone {boneName=spine_03}
   }
   Save cached pose 'CRCache_Dyn'(Pose=Component To Local(ComponentPose=@n12)) {cacheName=CRCache_Dyn}
   n13 = "Transform (Modify) Bone - Bone: base_collar_bk"(
    ComponentPose=Local To Component(LocalPose=Use cached pose 'CRCache_ControlRig'())
    Alpha=Get isSkyDive() -> (isSkyDive:double)
    Translation="-1,0,6"
    Rotation="12,0,0"
    Scale="1,1,1"
   ) {
    rotationMode=BMM_Additive
    rotationSpace=BCS_BoneSpace
    translationMode=BMM_Additive
    translationSpace=BCS_BoneSpace
    alphaBoolBlend {blendOption=Linear}
    alphaScaleBiasClamp {bInterpResult=true}
    boneToModify {boneName=base_collar_bk}
    rotation {pitch=12}
    translation {x=-1, z=6}
   }
   n14 = "Transform (Modify) Bone - Bone: base_collar_bk"(ComponentPose=@n13, Alpha=Get isCrouchedAlpha() -> (isCrouchedAlpha:double), Translation="-1,0,3", Rotation="20,0,0", Scale="1,1,1") {
    rotationMode=BMM_Additive
    rotationSpace=BCS_BoneSpace
    translationMode=BMM_Additive
    translationSpace=BCS_BoneSpace
    alphaBoolBlend {blendOption=Linear}
    alphaScaleBiasClamp {bInterpResult=true}
    boneToModify {boneName=base_collar_bk}
    rotation {pitch=20}
    translation {x=-1, z=3}
   }
   n15 = "Transform (Modify) Bone - Bone: base_collar_bk"(ComponentPose=@n14, Alpha=Get isRunning() -> (isRunning:double), Translation="-1,0,3", Rotation="20,0,0", Scale="1,1,1") {
    rotationMode=BMM_Additive
    rotationSpace=BCS_BoneSpace
    translationMode=BMM_Additive
    translationSpace=BCS_BoneSpace
    alphaBoolBlend {blendOption=Linear}
    alphaScaleBiasClamp {bInterpResult=true}
    boneToModify {boneName=base_collar_bk}
    rotation {pitch=20}
    translation {x=-1, z=3}
   }
   n16 = Get isLookup() -> (isLookup:double)
   n17 = "Transform (Modify) Bone - Bone: base_collar_bk"(ComponentPose=@n15, Alpha=@n16, Translation="0,0,2", Rotation="50,0,0", Scale="1,1,1") {
    rotationMode=BMM_Additive
    rotationSpace=BCS_BoneSpace
    translationMode=BMM_Additive
    translationSpace=BCS_BoneSpace
    alphaBoolBlend {blendOption=Linear}
    alphaScaleBiasClamp {bInterpResult=true, interpSpeedDecreasing=6, interpSpeedIncreasing=7}
    boneToModify {boneName=base_collar_bk}
    rotation {pitch=50}
    translation {z=2}
   }
   n18 = "Transform (Modify) Bone - Bone: base_collar_side_r"(ComponentPose=@n17, Alpha=@n16, Translation="0,0,0", Rotation="30,0,0", Scale="1,1,1") {
    rotationMode=BMM_Additive
    rotationSpace=BCS_BoneSpace
    translationMode=BMM_Additive
    translationSpace=BCS_BoneSpace
    alphaBoolBlend {blendOption=Linear}
    alphaScaleBiasClamp {bInterpResult=true, interpSpeedDecreasing=6, interpSpeedIncreasing=7}
    boneToModify {boneName=base_collar_side_r}
    rotation {pitch=30}
   }
   n19 = Get neckFixMeele() -> (neckFixMeele:double)
   n20 = "Transform (Modify) Bone - Bone: base_collar_side_l"(ComponentPose=@n18, Alpha=@n19, Translation="0,0,3", Rotation="20,0,0", Scale="1,1,1") {
    rotationMode=BMM_Additive
    rotationSpace=BCS_BoneSpace
    scaleMode=BMM_Additive
    scaleSpace=BCS_BoneSpace
    translationMode=BMM_Additive
    translationSpace=BCS_BoneSpace
    alphaBoolBlend {blendOption=Linear}
    alphaScaleBiasClamp {bInterpResult=true, interpSpeedDecreasing=20, interpSpeedIncreasing=20}
    boneToModify {boneName=base_collar_side_l}
    rotation {pitch=20}
    translation {z=3}
   }
   n21 = "Transform (Modify) Bone - Bone: base_collar_fr_l"(ComponentPose=@n20, Alpha=@n19, Translation="2,0,1", Rotation="0,0,-5", Scale="1,1,1") {
    rotationMode=BMM_Additive
    rotationSpace=BCS_BoneSpace
    translationMode=BMM_Additive
    translationSpace=BCS_BoneSpace
    alphaBoolBlend {blendOption=Linear}
    alphaScaleBiasClamp {bInterpResult=true, interpSpeedDecreasing=15, interpSpeedIncreasing=15}
    boneToModify {boneName=base_collar_fr_l}
    rotation {roll=-5}
    translation {x=2, z=1}
   }
   n22 = "Transform (Modify) Bone - Bone: base_collar_bk"(ComponentPose=@n21, Alpha=@n19, Translation="0,-2,0", Rotation="5,0,0", Scale="1,1,1") {
    rotationMode=BMM_Additive
    rotationSpace=BCS_BoneSpace
    translationMode=BMM_Additive
    translationSpace=BCS_BoneSpace
    alphaBoolBlend {blendOption=Linear}
    alphaScaleBiasClamp {bInterpResult=true, interpSpeedIncreasing=20}
    boneToModify {boneName=base_collar_bk}
    rotation {pitch=5}
    translation {y=-2}
   }
   n23 = Get irParachuting() -> (irParachuting:double)
   n24 = "Transform (Modify) Bone - Bone: base_collar_side_r"(ComponentPose=@n22, Alpha=@n23, Translation="0,0,-4", Rotation="-50,0,-10", Scale="1,1,1") {
    rotationMode=BMM_Additive
    rotationSpace=BCS_BoneSpace
    translationMode=BMM_Additive
    translationSpace=BCS_BoneSpace
    alphaBoolBlend {blendOption=Linear}
    alphaScaleBiasClamp {bInterpResult=true, interpSpeedDecreasing=15, interpSpeedIncreasing=15}
    boneToModify {boneName=base_collar_side_r}
    rotation {pitch=-50, roll=-10}
    translation {z=-4}
   }
   n25 = "Transform (Modify) Bone - Bone: base_collar_side_l"(ComponentPose=@n24, Alpha=@n23, Translation="0,0,-4", Rotation="-50,0,10", Scale="1,1,1") {
    rotationMode=BMM_Additive
    rotationSpace=BCS_BoneSpace
    translationMode=BMM_Additive
    translationSpace=BCS_BoneSpace
    alphaBoolBlend {blendOption=Linear}
    alphaScaleBiasClamp {bInterpResult=true, interpSpeedDecreasing=15, interpSpeedIncreasing=15}
    boneToModify {boneName=base_collar_side_l}
    rotation {pitch=-50, roll=10}
    translation {z=-4}
   }
   n26 = "Transform (Modify) Bone - Bone: base_collar_fr_r"(ComponentPose=@n25, Alpha=@n23, Translation="0,0,-1", Rotation="-10,0,0", Scale="1,1,1") {
    rotationMode=BMM_Additive
    rotationSpace=BCS_BoneSpace
    translationMode=BMM_Additive
    translationSpace=BCS_BoneSpace
    alphaBoolBlend {blendOption=Linear}
    alphaScaleBiasClamp {bInterpResult=true, interpSpeedDecreasing=15, interpSpeedIncreasing=15}
    boneToModify {boneName=base_collar_fr_r}
    rotation {pitch=-10}
    translation {z=-1}
   }
   n27 = "Transform (Modify) Bone - Bone: base_collar_fr_l"(ComponentPose=@n26, Alpha=@n23, Translation="0,0,-0.06", Rotation="-10,0,20", Scale="1,1,1") {
    rotationMode=BMM_Additive
    rotationSpace=BCS_BoneSpace
    translationMode=BMM_Additive
    translationSpace=BCS_BoneSpace
    alphaBoolBlend {blendOption=Linear}
    alphaScaleBiasClamp {bInterpResult=true, interpSpeedDecreasing=15, interpSpeedIncreasing=15}
    boneToModify {boneName=base_collar_fr_l}
    rotation {pitch=-10, roll=20}
    translation {z=-0.06}
   }
   Save cached pose 'CRCache_NeckFix'(Pose=Component To Local(ComponentPose=@n27)) {cacheName=CRCache_NeckFix}
   n28 = Get isRunning()
   n29 = "Transform (Modify) Bone - Bone: base_jacket_bk"(
    ComponentPose=Local To Component(LocalPose=Use cached pose 'CRCache_NeckFix'())
    Alpha=@n28
    Translation="2.5,0,1"
    Rotation="-5,0,0"
    Scale="1,1,1"
   ) {
    rotationMode=BMM_Additive
    rotationSpace=BCS_BoneSpace
    translationMode=BMM_Additive
    translationSpace=BCS_BoneSpace
    alphaBoolBlend {blendOption=Linear}
    alphaScaleBiasClamp {bInterpResult=true}
    boneToModify {boneName=base_jacket_bk}
    rotation {pitch=-5}
    translation {x=2.5, z=1}
   }
   n30 = "Transform (Modify) Bone - Bone: base_jacket_side_a_l"(ComponentPose=@n29, Alpha=@n28, Translation="2.5,0,0.5", Rotation="0,0,0", Scale="1,1,1") {
    rotationMode=BMM_Additive
    rotationSpace=BCS_BoneSpace
    translationMode=BMM_Additive
    translationSpace=BCS_BoneSpace
    alphaBoolBlend {blendOption=Linear}
    alphaScaleBiasClamp {bInterpResult=true, interpSpeedDecreasing=20, interpSpeedIncreasing=1}
    boneToModify {boneName=base_jacket_side_a_l}
    translation {x=2.5, z=0.5}
   }
   n31 = "Transform (Modify) Bone - Bone: base_jacket_side_a_r"(ComponentPose=@n30, Alpha=@n28, Translation="2.5,0,0.5", Rotation="0,0,0", Scale="1,1,1") {
    rotationMode=BMM_Additive
    rotationSpace=BCS_BoneSpace
    translationMode=BMM_Additive
    translationSpace=BCS_BoneSpace
    alphaBoolBlend {blendOption=Linear}
    alphaScaleBiasClamp {bInterpResult=true, interpSpeedDecreasing=20, interpSpeedIncreasing=1}
    boneToModify {boneName=base_jacket_side_a_r}
    translation {x=2.5, z=0.5}
   }
   n32 = "Transform (Modify) Bone - Bone: base_jacket_fr_a_r"(ComponentPose=@n31, Alpha=Get isCrouchedAlpha(), Translation="0,0,-2", Rotation="-10,0,0", Scale="1,1,1") {
    rotationMode=BMM_Additive
    rotationSpace=BCS_BoneSpace
    translationMode=BMM_Additive
    translationSpace=BCS_BoneSpace
    alphaBoolBlend {blendOption=Linear}
    alphaScaleBiasClamp {bInterpResult=true}
    boneToModify {boneName=base_jacket_fr_a_r}
    rotation {pitch=-10}
    translation {z=-2}
   }
   Save cached pose 'CRCache_JacketFix'(Pose=Component To Local(ComponentPose=@n32)) {cacheName=CRCache_JacketFix}
  ]}
 EventGraph {nodes [
   Event BlueprintUpdateAnimation()
   > n1 = Sequence!(execute) -> (then_0, then_1, then_2:Exec, then_3:Exec, then_4:Exec, then_5:Exec)
   > n2 = Gravity Override State Function DEPRECATED(CharacterPartAnimInstance=Self-Reference(), GameTimeInSeconds=GetGameTimeInSeconds(), GravityOverrideParams=Get Gravity Override Params())
   > Set GravityOverride(GravityOverride=@n2)
   n3 = NOT Boolean(A=Get bIsParachuteOpened())
   n4 = OR Boolean!(A=Get bIsSkydiveDiveMode(), B=Get bIsSkyDiving(), C:bool=Get bIsSkydiveFloating()) -> (ReturnValue)
   Set isSkyDive(execute=@n1.then_1, isSkyDive:double="To Float (Boolean)"(InBool=AND Boolean(A=@n4, B=@n3))) -> (Output_Get:double)
   n5 = Get bIsCrouchSprinting()
   n6 = Get bIsCrouching()
   n7 = Get bVehicleSeatForcesCrouch() -> (bVehicleSeatForcesCrouch:bool)
   > Set isCrouchedAlpha(isCrouchedAlpha:double="To Float (Boolean)"(InBool=OR Boolean(A=@n6, B=@n5, C=@n7))) -> (Output_Get:double)
   > Set irParachuting(irParachuting:double="To Float (Boolean)"(InBool=Get bIsParachuteOpened())) -> (Output_Get:double)
   n8 = Get bIsCrouching()
   n9 = Get bIsSkydiveFloating()
   n10 = Get bIsTacticalSprinting()
   n11 = GetOwnerSkeletalMeshComponent()
   n12 = BreakRotator(InRot=BreakTransform(InTransform=GetDeltaTransformFromRefPose(self=@n11, BoneName=head, BaseName=neck_02)).Rotation)
   n13 = Branch(
    execute=@n1.then_2
    Condition=AND Boolean!(
     A="float <= float"(A=@n12.Yaw, B=-13) -> (ReturnValue=~)
     B=NOR Boolean!(A:bool=@n10, B:bool=@n9) -> (ReturnValue:bool)
     C:bool=NOT Boolean(A=@n8)
    ) -> (ReturnValue)
   )
   > Set isLookup(isLookup:double=1) -> (Output_Get:double)
   Set isLookup(execute=@n13.else, isLookup=0)
   n14 = BreakTransform(InTransform=GetDeltaTransformFromRefPose(self=GetOwnerSkeletalMeshComponent(), BoneName=neck_01, BaseName=spine_05))
   n15 = Branch(execute=@n1.then_3, Condition="float >= float"(A=BreakRotator(InRot=@n14.Rotation), B=15))
   > Set HeadRotateL(HeadRotateL:bool=true) -> (Output_Get:bool)
   Set HeadRotateL(execute=@n15.else, HeadRotateL=false)
   n16 = Branch(execute=@n1.then_4, Condition=AND Boolean!(A=Get bIsPlayingMeleeAnim(), B=Get HeadRotateL() -> (HeadRotateL:bool)) -> (ReturnValue))
   > Set neckFixMeele(neckFixMeele:double=1) -> (Output_Get:double)
   Set neckFixMeele(execute=@n16.else, neckFixMeele=0)
  ]}
]
variables [
 Gravity Override Params {type=GravityOverrideParamsStruct}
 GravityOverride {type=Vector}
 HeadRotateL {type=boolean}
 irParachuting {type=number}
 isCrouchedAlpha {type=number}
 isCrouchRunning {type=number}
 isEmoting {type=number}
 isLookup {type=number}
 isRunning {type=number}
 isSkyDive {type=number}
 IsSwitchWeapon {type=number}
 neckFixMeele {type=number}
 SkelMesh {type=SkeletalMeshComponent}
]
