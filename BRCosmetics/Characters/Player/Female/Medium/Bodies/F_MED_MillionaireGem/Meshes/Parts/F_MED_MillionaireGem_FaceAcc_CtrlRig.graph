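# Control Rig for the F_MED_MillionaireGem face accessory part that adjusts right-side hair bones according to the pose of the right arm.
# A spherical pose reader on upperarm_r outputs a falloff weight that drives a Modify Transforms unit, blending corrective rotations onto bones including dyn_hair_a01_r and base_hair_b01_r.
# The hierarchy also holds nose ring, hair bun, left hair and collar bones, and the pose reader has debug drawing enabled.
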ControlRigBlueprint /BRCosmetics/Characters/Player/Female/Medium/Bodies/F_MED_MillionaireGem/Meshes/Parts/F_MED_MillionaireGem_FaceAcc_CtrlRig
graphs [
 RigVMFunctionLibrary {links [], nodes []}
 RigVMModel {nodes [
   RigUnit_BeginExecution = &RigUnit_BeginExecution()
   n1 = &RigUnit_SphericalPoseReader(
    ActiveRegionScaleFactors=(PositiveWidth=1,NegativeWidth=1,PositiveHeight=1,NegativeHeight=1)
    ActiveRegionSize=0.08
    Debug=(bDrawDebug=True,bDraw2D=False,bDrawLocalAxes=False,DebugScale=25,DebugSegments=20,DebugThickness=0.25)
    DriverAxis=(X=-1,Y=0,Z=0)
    DriverItem=(Type=Bone,Name="upperarm_r")
    FalloffRegionScaleFactors=(PositiveWidth=0.5,NegativeWidth=0,PositiveHeight=0.5,NegativeHeight=0)
    FalloffSize=0.3
    FlipHeightScaling=False
    FlipWidthScaling=False
    OptionalParentItem=(Type=Bone,Name="None")
    RotationOffset=(X=-105,Y=0,Z=-10)
    ExecutePin=@RigUnit_BeginExecution.ExecutePin
   )
   > &RigUnit_ModifyTransforms(
    ItemToModify="((Item=(Type=Bone,Name=\"dyn_hair_a01_r\"),Transform=(Rotation=(X=0.0227,Y=0.1722,Z=-0.1285,W=0.9764),Translation=(X=0,Y=0,Z=0),Scale3D=(X=1,Y=1,Z=1))),(Item=(Type=Bone,Name=\"base_hair_b01_r\"),Transform=(Rotation=(X=0.0302,Y=0.171,Z=-0.171,W=0.9698),Translat ...(truncated)"
    WeightMinimum=0
    Weight=@n1.OutputParam
   )
  ]}
]
hierarchy [
 root {loc=<0 0 0>, rot=<0 0 0 1>}
 pelvis {parent=root, loc=<0 -0.5301 98.0571>, rot=<0.0127 0.707 -0.0127 -0.707>}
 spine_01 {parent=pelvis, loc=<6.1208 -0.9328 0>, rot=<0 0 0.034 -0.9994>}
 spine_02 {parent=spine_01, loc=<6.5996 -0.2095 0>, rot=<0 0 0.0238 -0.9997>}
 spine_03 {parent=spine_02, loc=<6.4914 -0.1237 0>, rot=<0 0 0.0149 -0.9999>}
 spine_04 {parent=spine_03, loc=<6.9598 0.0794 0>, rot=<0 0 -0.0132 -0.9999>}
 spine_05 {parent=spine_04, loc=<6.9987 0 0>, rot=<0 0 -0.0015 -1>}
 neck_01 {parent=spine_05, loc=<11.3138 0.1529 0>, rot=<0 0 -0.2003 -0.9797>}
 neck_02 {parent=neck_01, loc=<4.5627 0 0>, rot=<0 0 0.0504 -0.9987>}
 head {parent=neck_02, loc=<5.3914 0 0>, rot=<0 0 0.115 -0.9934>}
 base_noseRing_m {parent=head, loc=<0.0836 8.5492 0>, rot=<0.0894 0.7014 0.7014 0.0894>}
 dyn_noseRing_m {parent=base_noseRing_m, loc=<0 0 0>, rot=<0 0 0 1>}
 base_hair_fr01_l {parent=head, loc=<8.9648 9.5706 -2.3989>, rot=<0.0239 -0.8842 -0.3876 -0.2598>}
 dyn_hair_fr01_l {parent=base_hair_fr01_l, loc=<0 0 0>, rot=<0 0 0 1>}
 dyn_hair_fr02_l {parent=dyn_hair_fr01_l, loc=<4.2251 0 0>, rot=<-0.0266 -0.0515 0.0941 -0.9939>}
 dyn_hair_fr03_l {parent=dyn_hair_fr02_l, loc=<3.9635 0 0>, rot=<-0.0365 -0.2504 -0.0115 -0.9674>}
 dyn_hair_fr04_l {parent=dyn_hair_fr03_l, loc=<3.4669 0 0>, rot=<-0.0564 -0.1539 -0.0612 -0.9846>}
 base_hair_side_l {parent=head, loc=<5.9421 4.2998 -7.0633>, rot=<0.1809 -0.9655 -0.1282 -0.1362>}
 dyn_hair_side_l {parent=base_hair_side_l, loc=<0 0 0>, rot=<0 0 0 1>}
 dyn_hair_a01_l {parent=dyn_hair_side_l, loc=<3.9218 1.3245 -1.023>, rot=<-0.2739 -0.1832 -0.0079 -0.9441>}
 dyn_hair_a02_l {parent=dyn_hair_a01_l, loc=<4.3347 0 0>, rot=<0.0167 -0.1291 -0.0707 -0.989>}
 dyn_hair_a03_l {parent=dyn_hair_a02_l, loc=<4.0775 0 0>, rot=<0.0102 0.0966 -0.1326 -0.9864>}
 dyn_hair_a04_l {parent=dyn_hair_a03_l, loc=<3.6981 0 0>, rot=<-0.1059 -0.3599 -0.01 -0.9269>}
 dyn_hair_a05_l {parent=dyn_hair_a04_l, loc=<2.9347 0 0>, rot=<0 0 0 1>}
 dyn_hair_b01_l {parent=dyn_hair_side_l, loc=<3.9959 -1.9618 -1.1911>, rot=<-0.1046 -0.213 -0.0467 -0.9703>}
 dyn_hair_b02_l {parent=dyn_hair_b01_l, loc=<4.524 0 0>, rot=<0.0469 -0.0856 -0.2195 -0.9707>}
 dyn_hair_b03_l {parent=dyn_hair_b02_l, loc=<3.8317 0 0>, rot=<0 0 0 1>}
 base_hairBun_r {parent=head, loc=<11.3151 -4.3318 5.6192>, rot=<0.795 0.0021 0.4364 -0.4214>}
 dyn_hairBun_r {parent=base_hairBun_r, loc=<0 0 0>, rot=<0 0 0 1>}
 base_hairBun_l {parent=head, loc=<11.3026 -4.3555 -5.6306>, rot=<0.4347 -0.4423 0.0071 -0.7844>}
 dyn_hairBun_l {parent=base_hairBun_l, loc=<0 0 0>, rot=<0 0 0 1>}
 base_hair_side_r {parent=head, loc=<4.8613 4.1803 6.1356>, rot=<0 0 1 -0.0041>}
 dyn_hair_side_r {parent=base_hair_side_r, loc=<0 0 0>, rot=<0 0 0 1>}
 dyn_hair_a01_r {parent=dyn_hair_side_r, loc=<4.7057 0 0>, rot=<0 0 0 1>}
 dyn_hair_a02_r {parent=dyn_hair_a01_r, loc=<4.5875 0 0>, rot=<0 0 0 1>}
 dyn_hair_a03_r {parent=dyn_hair_a02_r, loc=<5.1858 0 0>, rot=<0 0 0 1>}
 dyn_hair_a04_r {parent=dyn_hair_a03_r, loc=<2.0873 -1.2692 -3.492>, rot=<0.2893 -0.6984 0.3514 -0.5523>}
 dyn_hair_a05_r {parent=dyn_hair_a04_r, loc=<3.2349 0 0>, rot=<0.0793 -0.2326 0.0328 -0.9688>}
 base_hair_b01_r {parent=base_hair_side_r, loc=<10.6887 -2.3314 -0.1823>, rot=<-0.1554 -0.261 0.0506 -0.9514>}
 dyn_hair_b01_r {parent=base_hair_b01_r, loc=<0 0 0>, rot=<0 0 0 1>}
 base_hair_fr01_r {parent=head, loc=<9.5035 9.0044 0.9928>, rot=<0.4495 0.3568 0.8189 0.0075>}
 dyn_hair_fr01_r {parent=base_hair_fr01_r, loc=<0 0 0>, rot=<0 0 0 1>}
 dyn_hair_fr02_r {parent=dyn_hair_fr01_r, loc=<2.5958 0 0>, rot=<-0.0029 -0.3004 -0.0072 -0.9538>}
 dyn_hair_fr03_r {parent=dyn_hair_fr02_r, loc=<3.4562 0 0>, rot=<-0.0346 -0.085 -0.0399 -0.995>}
 dyn_hair_fr04_r {parent=dyn_hair_fr03_r, loc=<2.5034 0 0>, rot=<0 0 0 1>}
 clavicle_l {parent=spine_05, loc=<5.4831 -0.1053 -3.2946>, rot=<0.0465 -0.7289 -0.009 -0.6829>}
 upperarm_l {parent=clavicle_l, loc=<9.8658 -1.2136 -0.4105>, rot=<0.0164 -0.3245 -0.0269 -0.9454>}
 clavicle_r {parent=spine_05, loc=<5.4831 -0.1053 3.2946>, rot=<0.7289 0.0465 -0.6829 0.009>}
 upperarm_r {parent=clavicle_r, loc=<-9.8658 1.2136 0.4105>, rot=<0.0164 -0.3245 -0.0269 -0.9454>}
 base_collar_fr_l {parent=spine_05, loc=<7.1995 5.6478 -7.8861>, rot=<0.9705 0.1685 -0.1427 0.0964>}
 dyn_collar_fr_l {parent=base_collar_fr_l, loc=<0 0 0>, rot=<0 0 0 1>}
 base_collar_side_l {parent=spine_05, loc=<10.9958 0.0516 -5.814>, rot=<0.95 0.1447 -0.1383 0.2395>}
 dyn_collar_side_l {parent=base_collar_side_l, loc=<0 0 0>, rot=<0 0 0 1>}
 base_collar_bk {parent=spine_05, loc=<12.9056 -4.0102 0>, rot=<-0.707 0.0093 0.0093 -0.707>}
 dyn_collar_bk_l {parent=base_collar_bk, loc=<0 0 0>, rot=<0 0 0 1>}
 base_collar_side_r {parent=spine_05, loc=<10.9959 0.0516 5.814>, rot=<-0.2395 0.1383 -0.1447 -0.95>}
 dyn_collar_side_r {parent=base_collar_side_r, loc=<0 0 0>, rot=<0 0 0 1>}
 base_collar_fr_r {parent=spine_05, loc=<7.199 5.6478 7.8861>, rot=<-0.0964 0.1427 -0.1685 -0.9705>}
 dyn_collar_fr_r {parent=base_collar_fr_r, loc=<0 0 0>, rot=<0 0 0 1>}
]
