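# Physics asset for the MillionaireGem female character's side-hair face accessory part, defining capsule bodies and joint limits for secondary hair motion.
# Front and side hair chains, branching into a/b strands, hang from base bones locked to the head, with limited swing and minimal twist.
# Head, neck, collar, right clavicle and upper-arm capsules supply body shapes around the hair; collar bodies have no constraints, while the upper-arm-to-clavicle joint rotates freely.
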PhysicsAsset /BRCosmetics/Characters/Player/Female/Medium/Bodies/F_MED_MillionaireGem/Meshes/Parts/F_MED_MillionaireGem_FaceAcc_hairSide__Physics
bodies {
 base_collar_fr_l [
  {
   shapeType=Capsule
   length=15.2499
   radius=2.3777
   shapeName=base_collar_fr_l_capsule
   center=<9.448 1.878 3.2714>
   rotation=<rot 0 18.5905 95.8503>
  }
  {
   shapeType=Capsule
   length=8.8383
   radius=3.7504
   shapeName=base_collar_fr_l_capsule
   center=<1.3919 -1.3653 2.6406>
   rotation=<rot -2.9479 5.0683 84.9003>
  }
 ]
 base_collar_side_l [{
   shapeType=Capsule
   length=8.2748
   radius=2.479
   shapeName=base_collar_side_l_capsule
   center=<2.599 -0.131 1.4641>
   rotation=<rot 0 0 102.9594>
  }]
 base_hair_fr01_l [{shapeType=Capsule, length=0.505, radius=0.505, shapeName=base_hair_fr01_l_capsule, rotation=<rot 0 -90 0>}]
 base_hair_fr01_r [{shapeType=Capsule, length=0.505, radius=0.505, shapeName=base_hair_fr01_r_capsule, rotation=<rot 0 -90 0>}]
 base_hair_side_l [{shapeType=Capsule, length=0.505, radius=0.505, shapeName=base_hair_side_l_capsule, rotation=<rot 0 -90 0>}]
 base_hair_side_r [{shapeType=Capsule, length=0.505, radius=0.505, shapeName=base_hair_side_r_capsule, rotation=<rot 0 -90 0>}]
 base_noseRing_m [{shapeType=Capsule, length=0.505, radius=0.505, shapeName=base_noseRing_m_capsule, rotation=<rot 0 -90 0>}]
 clavicle_r [{
   shapeType=Capsule
   length=6.5415
   radius=6.4203
   shapeName=clavicle_r_capsule
   center=<-11.5173 0 1.9573>
   rotation=<rot -0.0001 -90 96.6861>
  }]
 dyn_collar_fr_r [{
   shapeType=Capsule
   length=25.0141
   radius=4.3797
   shapeName=dyn_collar_fr_r_capsule
   center=<8.5232 2.4212 0.0619>
   rotation=<rot 72.0284 59.4501 -14.4987>
  }]
 dyn_collar_side_r [{shapeType=Capsule, length=2.138, radius=3.48, shapeName=dyn_collar_side_r_capsule, center=<0.6258 0 0.5066>, rotation=<rot 90 0 0>}]
 dyn_hair_a01_l [{shapeType=Capsule, length=2.8661, radius=1.1165, shapeName=dyn_hair_a01_l_capsule, center=<1.7434 0 0>, rotation=<rot 90 0 0>}]
 dyn_hair_a01_r [{shapeType=Capsule, length=3.3896, radius=1.1171, shapeName=dyn_hair_a01_r_capsule, center=<1.4907 0 0.0124>, rotation=<rot 90 0 0>}]
 dyn_hair_a02_l [{shapeType=Capsule, length=3.2091, radius=1.3821, shapeName=dyn_hair_a02_l_capsule, center=<1.4763 0 0>, rotation=<rot 90 0 0>}]
 dyn_hair_a02_r [{shapeType=Capsule, length=4.782, radius=1.1778, shapeName=dyn_hair_a02_r_capsule, center=<2.5692 0 0>, rotation=<rot 90 0 0>}]
 dyn_hair_a03_l [{shapeType=Capsule, length=2.7542, radius=1.1198, shapeName=dyn_hair_a03_l_capsule, center=<1.3172 0 0>, rotation=<rot 90 0 0>}]
 dyn_hair_a03_r [{shapeType=Capsule, length=5.958, radius=1.0468, shapeName=dyn_hair_a03_r_capsule, center=<2.6547 0 -1.579>, rotation=<rot 61.2982 -0.0001 0>}]
 dyn_hair_a04_l [{
   shapeType=Capsule
   length=2.1752
   radius=1.0047
   shapeName=dyn_hair_a04_l_capsule
   center=<0.5427 0 -0.1146>
   rotation=<rot 85.2778 179.9997 179.9997>
  }]
 dyn_hair_a04_r [{shapeType=Capsule, length=1.7118, radius=1.4476, shapeName=dyn_hair_a04_r_capsule, center=<0.8742 0 0>, rotation=<rot 90 0 0>}]
 dyn_hair_a05_l [{shapeType=Capsule, length=2.6479, radius=0.6767, shapeName=dyn_hair_a05_l_capsule, center=<0.287 0 0>, rotation=<rot 47.677 0 0>}]
 dyn_hair_a05_r [{shapeType=Capsule, length=1.8263, radius=0.9988, shapeName=dyn_hair_a05_r_capsule, center=<0.7571 0 0>, rotation=<rot 90 0 0>}]
 dyn_hair_b01_l [{shapeType=Capsule, length=3.228, radius=0.7707, shapeName=dyn_hair_b01_l_capsule, center=<1.2762 0 0>, rotation=<rot 90 0 0>}]
 dyn_hair_b01_r [{shapeType=Capsule, length=1, radius=0.5, shapeName=dyn_hair_b01_r_capsule, rotation=<rot 90 0 0>}]
 dyn_hair_b02_l [{shapeType=Capsule, length=3.228, radius=0.7707, shapeName=dyn_hair_b02_l_capsule, center=<1.2762 0 0>, rotation=<rot 90 0 0>}]
 dyn_hair_b03_l [{shapeType=Capsule, length=3.228, radius=0.7707, shapeName=dyn_hair_b03_l_capsule, center=<1.2762 0 0>, rotation=<rot 90 0 0>}]
 dyn_hair_fr01_l [{shapeType=Capsule, length=3, radius=0.6, shapeName=dyn_hair_fr01_l_capsule, center=<1 0 0>, rotation=<rot 90 0 0>}]
 dyn_hair_fr01_r [{shapeType=Capsule, length=2.5, radius=0.5, shapeName=dyn_hair_fr01_r_capsule, center=<0.9892 0 0>, rotation=<rot 90 0 0>}]
 dyn_hair_fr02_l [{shapeType=Capsule, length=3, radius=0.6, shapeName=dyn_hair_fr02_l_capsule, center=<1 0 0>, rotation=<rot 90 0 0>}]
 dyn_hair_fr02_r [{shapeType=Capsule, length=2.8145, radius=0.5, shapeName=dyn_hair_fr02_r_capsule, center=<1.6303 0 0>, rotation=<rot 90 0 0>}]
 dyn_hair_fr03_l [{shapeType=Capsule, length=3, radius=0.6, shapeName=dyn_hair_fr03_l_capsule, center=<1 0 0>, rotation=<rot 90 0 0>}]
 dyn_hair_fr03_r [{shapeType=Capsule, length=2, radius=0.5, shapeName=dyn_hair_fr03_r_capsule, center=<1.2139 0 0>, rotation=<rot 90 0 0>}]
 dyn_hair_fr04_l [{shapeType=Capsule, length=3, radius=0.6, shapeName=dyn_hair_fr04_l_capsule, center=<1 0 0>, rotation=<rot 90 0 0>}]
 dyn_hair_fr04_r [{shapeType=Capsule, length=2, radius=0.5, shapeName=dyn_hair_fr04_r_capsule, center=<1.2422 0 0>, rotation=<rot 90 0 0>}]
 dyn_hair_side_l [{shapeType=Capsule, length=3.3132, radius=0.5, shapeName=dyn_hair_side_l_capsule, center=<1.196 0 0>, rotation=<rot 67.5444 0.014 -0.0001>}]
 dyn_hair_side_r [{shapeType=Capsule, length=3.435, radius=0.9906, shapeName=dyn_hair_side_r_capsule, center=<1.7092 0 0>, rotation=<rot 90 0 0>}]
 head [{shapeType=Capsule, length=1.6413, radius=6.6971, shapeName=head_capsule, center=<2.2192 -1.3585 0.1607>, rotation=<rot 90 0 0>}]
 neck_01 [{shapeType=Capsule, length=4.482, radius=5.3923, shapeName=neck_01_capsule, center=<0.9161 1.3997 0>, rotation=<rot 90 0 0>}]
 upperarm_r [
  {shapeType=Capsule, length=19, radius=4, shapeName=upperarm_r_capsule, center=<-13.5 0 0>, rotation=<rot 90 0 0>}
  {shapeType=Capsule, length=19, radius=4, shapeName=upperarm_r_capsule, center=<-13.5 -4.1402 -8.1787>, rotation=<rot 90 0 0>}
 ]
}
constraints [
 {bone1Name=dyn_hair_fr01_l, bone2Name=base_hair_fr01_l, swing1LimitDegrees=40, swing2LimitDegrees=45, twistLimitDegrees=2}
 {bone1Name=dyn_hair_fr02_l, bone2Name=dyn_hair_fr01_l, swing1LimitDegrees=12, swing2LimitDegrees=12, twistLimitDegrees=2}
 {bone1Name=dyn_hair_fr03_l, bone2Name=dyn_hair_fr02_l, swing1LimitDegrees=16, swing2LimitDegrees=16, twistLimitDegrees=2}
 {bone1Name=dyn_hair_fr04_l, bone2Name=dyn_hair_fr03_l, swing1LimitDegrees=30, swing2LimitDegrees=30, twistLimitDegrees=2}
 {bone1Name=dyn_hair_side_l, bone2Name=base_hair_side_l, swing1LimitDegrees=18, swing2LimitDegrees=10, twistLimitDegrees=2}
 {bone1Name=dyn_hair_a01_l, bone2Name=dyn_hair_side_l, swing1LimitDegrees=15, swing2LimitDegrees=12, twistLimitDegrees=2}
 {bone1Name=dyn_hair_a02_l, bone2Name=dyn_hair_a01_l, swing1LimitDegrees=15, swing2LimitDegrees=18, twistLimitDegrees=2}
 {bone1Name=dyn_hair_a03_l, bone2Name=dyn_hair_a02_l, swing1LimitDegrees=15, swing2LimitDegrees=15, twistLimitDegrees=2}
 {bone1Name=dyn_hair_a04_l, bone2Name=dyn_hair_a03_l, swing1LimitDegrees=9, swing2LimitDegrees=10, twistLimitDegrees=2}
 {bone1Name=dyn_hair_a05_l, bone2Name=dyn_hair_a04_l, swing1LimitDegrees=9, swing2LimitDegrees=10, twistLimitDegrees=2}
 {bone1Name=dyn_hair_b01_l, bone2Name=dyn_hair_side_l, swing1LimitDegrees=10, swing2LimitDegrees=10, twistLimitDegrees=2}
 {bone1Name=dyn_hair_b02_l, bone2Name=dyn_hair_b01_l, swing1LimitDegrees=10, swing2LimitDegrees=10, twistLimitDegrees=2}
 {bone1Name=dyn_hair_b03_l, bone2Name=dyn_hair_b02_l, swing1LimitDegrees=10, swing2LimitDegrees=10, twistLimitDegrees=2}
 {bone1Name=dyn_hair_side_r, bone2Name=base_hair_side_r, swing1LimitDegrees=12, swing2LimitDegrees=10, twistLimitDegrees=2}
 {bone1Name=dyn_hair_a01_r, bone2Name=dyn_hair_side_r, swing1LimitDegrees=8, swing2LimitDegrees=17, twistLimitDegrees=2}
 {bone1Name=dyn_hair_a02_r, bone2Name=dyn_hair_a01_r, swing1LimitDegrees=7, swing2LimitDegrees=17, twistLimitDegrees=2}
 {bone1Name=dyn_hair_a03_r, bone2Name=dyn_hair_a02_r, swing1LimitDegrees=7, swing2LimitDegrees=10, twistLimitDegrees=2}
 {bone1Name=dyn_hair_a04_r, bone2Name=dyn_hair_a03_r, swing1LimitDegrees=10, swing2LimitDegrees=14, twistLimitDegrees=2}
 {bone1Name=dyn_hair_a05_r, bone2Name=dyn_hair_a04_r, swing1LimitDegrees=10, swing2LimitDegrees=15, twistLimitDegrees=2}
 {bone1Name=dyn_hair_b01_r, bone2Name=dyn_hair_a01_r, swing1LimitDegrees=10, swing2LimitDegrees=10, twistLimitDegrees=2}
 {
  bone1Name=base_noseRing_m
  bone2Name=head
  swing1LimitDegrees=45
  swing1Motion=Locked
  swing2LimitDegrees=45
  swing2Motion=Locked
  twistLimitDegrees=45
  twistMotion=Locked
 }
 {
  bone1Name=base_hair_fr01_l
  bone2Name=head
  swing1LimitDegrees=45
  swing1Motion=Locked
  swing2LimitDegrees=45
  swing2Motion=Locked
  twistLimitDegrees=45
  twistMotion=Locked
 }
 {
  bone1Name=base_hair_side_r
  bone2Name=head
  swing1LimitDegrees=45
  swing1Motion=Locked
  swing2LimitDegrees=45
  swing2Motion=Locked
  twistLimitDegrees=45
  twistMotion=Locked
 }
 {
  bone1Name=head
  bone2Name=neck_01
  swing1LimitDegrees=45
  swing1Motion=Locked
  swing2LimitDegrees=45
  swing2Motion=Locked
  twistLimitDegrees=45
  twistMotion=Locked
 }
 {
  bone1Name=upperarm_r
  bone2Name=clavicle_r
  swing1LimitDegrees=45
  swing1Motion=Free
  swing2LimitDegrees=45
  swing2Motion=Free
  twistLimitDegrees=45
  twistMotion=Free
 }
 {
  bone1Name=base_hair_fr01_r
  bone2Name=head
  swing1LimitDegrees=45
  swing1Motion=Locked
  swing2LimitDegrees=45
  swing2Motion=Locked
  twistLimitDegrees=45
  twistMotion=Locked
 }
 {bone1Name=dyn_hair_fr01_r, bone2Name=base_hair_fr01_r, swing1LimitDegrees=20, swing2LimitDegrees=20, twistLimitDegrees=0}
 {bone1Name=dyn_hair_fr02_r, bone2Name=dyn_hair_fr01_r, swing1LimitDegrees=20, swing2LimitDegrees=20, twistLimitDegrees=0}
 {bone1Name=dyn_hair_fr03_r, bone2Name=dyn_hair_fr02_r, swing1LimitDegrees=20, swing2LimitDegrees=20, twistLimitDegrees=0}
 {bone1Name=dyn_hair_fr04_r, bone2Name=dyn_hair_fr03_r, swing1LimitDegrees=20, swing2LimitDegrees=20, twistLimitDegrees=0}
 {
  bone1Name=base_hair_side_l
  bone2Name=head
  swing1LimitDegrees=45
  swing1Motion=Locked
  swing2LimitDegrees=45
  swing2Motion=Locked
  twistLimitDegrees=45
  twistMotion=Locked
 }
]
