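# Post-process-style hair animation for the MillionaireGem female character's FaceAcc part, copying the parent mesh pose.
# Constraint and head-driven Look At feed two RigidBody sims, switching to hair-side physics when crouch-aiming with raised arm yaw.
# Control Rig applies while targeting, then additive bone offsets adjust hair, some driven by head angle or open parachute.
# Event graph computes those states and gravity override via a deprecated function; bun AnimDynamics nodes are disconnected.
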AnimBlueprint /BRCosmetics/Characters/Player/Female/Medium/Bodies/F_MED_MillionaireGem/Meshes/Parts/F_MED_MillionaireGem_FaceAcc_AnimBP
target_skeleton=&/F_MED_MillionaireGem_FaceAcc_Skeleton
graphs [
 AnimGraph {nodes [
   n1 = Get GravityOverride()
   n2 = Anim Dynamics - Bone: dyn_hairBun_r!(ComponentPose, Alpha=0.2, GravityOverride:Vector=@n1, ExternalForce="0,0,0") -> (Pose) {
    angularSpringConstant=10000
    bAngularSpring=true
    bUseGravityOverride=true
    simulationSpace=BoneRelative
    alphaBoolBlend {blendOption=Linear}
    boundBone {boneName=dyn_hairBun_r}
    componentAppliedLinearAccClamp=<3000 3000 3000>
    componentLinearAccScale=<0.3 0.3 0.3>
    componentLinearVelScale=<0.3 0.3 0.3>
    physicsBodyDefinitions [{
      boundBone {boneName=dyn_hairBun_r}
      boxExtents=<20 20 20>
      constraintSetup {angularLimitsMax=<0 5 5>, angularLimitsMin=<0 -5 -5>, angularTarget=<0 0 0>}
      localJointOffset=<0.5 0 0>
     }]
    relativeSpaceBone {boneName=spine_05}
   }
   Anim Dynamics - Bone: dyn_hairBun_l!(ComponentPose=@n2, Alpha=0.2, GravityOverride:Vector=@n1, ExternalForce="0,0,0") -> (Pose) {
    angularSpringConstant=10000
    bAngularSpring=true
    bUseGravityOverride=true
    simulationSpace=BoneRelative
    alphaBoolBlend {blendOption=Linear}
    boundBone {boneName=dyn_hairBun_l}
    componentAppliedLinearAccClamp=<3000 3000 3000>
    componentLinearAccScale=<0.3 0.3 0.3>
    componentLinearVelScale=<0.3 0.3 0.3>
    physicsBodyDefinitions [{
      boundBone {boneName=dyn_hairBun_l}
      boxExtents=<20 20 20>
      constraintSetup {angularLimitsMax=<0 5 5>, angularLimitsMin=<0 -5 -5>, angularTarget=<0 0 0>}
      localJointOffset=<0.5 0 0>
     }]
    relativeSpaceBone {boneName=spine_05}
   }
   n3 = "Transform (Modify) Bone - Bone: dyn_hair_a02_r"(
    ComponentPose=Local To Component(LocalPose=Use cached pose 'CRCache_CollarClip'())
    Alpha=1
    Translation="0,0,0"
    Rotation="-10,0,-10"
    Scale="1,1,1"
   ) {
    rotationMode=BMM_Additive
    rotationSpace=BCS_BoneSpace
    alphaBoolBlend {blendOption=Linear}
    boneToModify {boneName=dyn_hair_a02_r}
    rotation {pitch=-10, roll=-10}
   }
   n4 = "Transform (Modify) Bone - Bone: dyn_hair_a03_r"(ComponentPose=@n3, Alpha=1, Translation="0,0,0", Rotation="-25,0,-10", Scale="1,1,1") {
    rotationMode=BMM_Additive
    rotationSpace=BCS_BoneSpace
    alphaBoolBlend {blendOption=Linear}
    boneToModify {boneName=dyn_hair_a03_r}
    rotation {pitch=-25, roll=-10}
   }
   n5 = "Transform (Modify) Bone - Bone: dyn_hair_side_r"(ComponentPose=@n4, Alpha=1, Translation="0,0,-0.5", Rotation="-4,0,0", Scale="1,1,1") {
    rotationMode=BMM_Additive
    rotationSpace=BCS_BoneSpace
    translationMode=BMM_Additive
    translationSpace=BCS_BoneSpace
    alphaBoolBlend {blendOption=Linear}
    boneToModify {boneName=dyn_hair_side_r}
    rotation {pitch=-4}
    translation {z=-0.5}
   }
   n6 = "Transform (Modify) Bone - Bone: dyn_hair_a01_r"(ComponentPose=@n5, Alpha=1, Translation="0,0,-0.2", Rotation="-2,0,0", Scale="1,1,1") {
    rotationMode=BMM_Additive
    rotationSpace=BCS_BoneSpace
    translationMode=BMM_Additive
    translationSpace=BCS_BoneSpace
    alphaBoolBlend {blendOption=Linear}
    boneToModify {boneName=dyn_hair_a01_r}
    rotation {pitch=-2}
    translation {z=-0.2}
   }
   n7 = Get AimUp() -> (AimUp:double)
   n8 = "Transform (Modify) Bone - Bone: dyn_hair_b01_l"(ComponentPose=@n6, Alpha=@n7, Translation="0,0,2", Rotation="0,0,0", Scale="1,1,1") {
    translationMode=BMM_Additive
    translationSpace=BCS_BoneSpace
    alphaBoolBlend {blendOption=Linear}
    alphaScaleBiasClamp {bInterpResult=true, interpSpeedDecreasing=5, interpSpeedIncreasing=5}
    boneToModify {boneName=dyn_hair_b01_l}
    translation {z=2}
   }
   n9 = "Transform (Modify) Bone - Bone: base_hair_side_r"(ComponentPose=@n8, Alpha=@n7, Translation="0,0,1", Rotation="0,0,0", Scale="1,1,1") {
    rotationMode=BMM_Additive
    rotationSpace=BCS_BoneSpace
    translationMode=BMM_Additive
    translationSpace=BCS_BoneSpace
    alphaBoolBlend {blendOption=Linear}
    alphaScaleBiasClamp {bInterpResult=true, interpSpeedDecreasing=5, interpSpeedIncreasing=5}
    boneToModify {boneName=base_hair_side_r}
    translation {z=1}
   }
   n10 = "Transform (Modify) Bone - Bone: dyn_hair_a03_r"(ComponentPose=@n9, Alpha=@n7, Translation="0,-3,0", Rotation="0,0,0", Scale="1,1,1") {
    rotationMode=BMM_Additive
    rotationSpace=BCS_BoneSpace
    translationMode=BMM_Additive
    translationSpace=BCS_BoneSpace
    alphaBoolBlend {blendOption=Linear}
    alphaScaleBiasClamp {bInterpResult=true, interpSpeedDecreasing=5, interpSpeedIncreasing=5}
    boneToModify {boneName=dyn_hair_a03_r}
    translation {y=-3}
   }
   n11 = "Transform (Modify) Bone - Bone: dyn_hair_a03_l"(ComponentPose=@n10, Alpha=@n7, Translation="0,2,-1.5", Rotation="0,0,0", Scale="1,1,1") {
    rotationMode=BMM_Additive
    rotationSpace=BCS_BoneSpace
    translationMode=BMM_Additive
    translationSpace=BCS_BoneSpace
    alphaBoolBlend {blendOption=Linear}
    alphaScaleBiasClamp {bInterpResult=true, interpSpeedDecreasing=5, interpSpeedIncreasing=5}
    boneToModify {boneName=dyn_hair_a03_l}
    translation {y=2, z=-1.5}
   }
   n12 = "Transform (Modify) Bone - Bone: base_hair_side_l"(ComponentPose=@n11, Alpha="To Float (Boolean)"(InBool=Get bIsParachuteOpened()), Translation="0,0,0", Rotation="0,15,0", Scale="1,1,1") {
    rotationMode=BMM_Additive
    rotationSpace=BCS_BoneSpace
    translationMode=BMM_Additive
    translationSpace=BCS_BoneSpace
    alphaBoolBlend {blendOption=Linear}
    alphaScaleBiasClamp {interpSpeedDecreasing=5, interpSpeedIncreasing=5}
    boneToModify {boneName=base_hair_side_l}
    rotation {yaw=15}
   }
   Output Pose(Result=Component To Local(ComponentPose=@n12)) {layerGroup=None}
   n13 = "Constraint - base_hair_b01_r (P)"(ComponentPose=Local To Component(LocalPose=Copy Pose From Mesh() {bUseAttachedParent=true}), Alpha=1, ConstraintWeights_0=1) {
    alphaBoolBlend {blendOption=Linear}
    boneToModify {boneName=base_hair_b01_r}
    constraintSetup [{targetBone {boneName=dyn_hair_a01_r}}]
    constraintWeights [1]
   }
   n14 = Get Look up Alpha() -> (Look up Alpha:double)
   n15 = Look At - Bone: base_hair_side_r(ComponentPose=@n13, Alpha=@n14) {
    alphaBoolBlend {blendOption=Linear}
    boneToModify {boneName=base_hair_side_r}
    lookAtLocation {x=0}
    lookAtTarget {boneReference {boneName=root}}
    lookAt_Axis {axis=<1 0>}
   }
   n16 = Look At - Bone: base_hair_side_l(ComponentPose=@n15, Alpha=@n14) {
    alphaBoolBlend {blendOption=Linear}
    alphaScaleBiasClamp {bInterpResult=true, interpSpeedDecreasing=5, interpSpeedIncreasing=5}
    boneToModify {boneName=base_hair_side_l}
    lookAtLocation {x=0}
    lookAtTarget {boneReference {boneName=root}}
    lookAt_Axis {axis=<1 0>}
   }
   Save cached pose 'CRCache_entry'(Pose=Component To Local(ComponentPose=@n16)) {cacheName=CRCache_entry}
   n17 = Get GravityOverride()
   n18 = RigidBody!(ComponentPose=Local To Component(LocalPose=Use cached pose 'CRCache_entry'()), Alpha, OverrideWorldGravity:Vector=@n17, ExternalForce) -> (Pose) {
    bOverrideWorldGravity=true
    overridePhysicsAsset=&/F_MED_MillionaireGem_FaceAcc_hairSide__Physics
    simulationSpace=BaseBoneSpace
    alphaBoolBlend {blendOption=Linear}
    baseBoneRef {boneName=spine_05}
    componentAppliedLinearAccClamp=<3000 3000 3000>
    componentLinearAccScale=<0.3 0.3 0.4>
    componentLinearVelScale=<0.3 0.3 0.4>
    simSpaceSettings {
     maxAngularAcceleration=1000
     maxAngularVelocity=3
     maxLinearAcceleration=700
     maxLinearVelocity=500
     velocityScaleZ=0
     externalLinearDragV=<0.2 0.2>
    }
   }
   n19 = RigidBody(ComponentPose=Local To Component(LocalPose=Use cached pose 'CRCache_entry'()), OverrideWorldGravity=@n17) {
    bOverrideWorldGravity=true
    overridePhysicsAsset=&/F_MED_MillionaireGem_FaceAcc_Physics
    simulationSpace=BaseBoneSpace
    alphaBoolBlend {blendOption=Linear}
    baseBoneRef {boneName=spine_05}
    componentAppliedLinearAccClamp=<3000 3000 3000>
    componentLinearAccScale=<0.3 0.3 0.4>
    componentLinearVelScale=<0.3 0.3 0.4>
    simSpaceSettings {
     maxAngularAcceleration=1000
     maxAngularVelocity=3
     maxLinearAcceleration=700
     maxLinearVelocity=500
     velocityScaleZ=0
     externalLinearDragV=<0.2 0.2>
    }
   }
   Save cached pose 'CRCache_RB'(Pose=Blend Poses by bool(
     BlendPose_0=Component To Local(ComponentPose=@n18)
     BlendPose_1=Component To Local(ComponentPose=@n19)
     BlendTime_0=0.2
     BlendTime_1=0.2
     bActiveValue=Get BlendPoseAlpha() -> (BlendPoseAlpha:bool)
    ) {blendType=Linear}) {cacheName=CRCache_RB}
   n20 = Control Rig(Source=Use cached pose 'CRCache_RB'()) {
    alphaBoolBlend {blendOption=Linear}
    controlRigAssetReference {blueprintRigClass=&/F_MED_MillionaireGem_FaceAcc_CtrlRig.F_MED_MillionaireGem_FaceAcc_CtrlRig_C}
   }
   Save cached pose 'CRCache_CollarClip'(Pose=Blend Poses by bool(BlendPose_0=@n20, BlendPose_1=Use cached pose 'CRCache_RB'(), BlendTime_0=0.2, BlendTime_1=0.2, bActiveValue=Get bIsTargeting()) {blendType=Linear}) {cacheName=CRCache_CollarClip}
  ]}
 EventGraph {nodes [
   Event BlueprintUpdateAnimation()
   > n1 = Sequence!(execute) -> (then_0, then_1, then_2:Exec, then_3:Exec, then_4:Exec, then_5:Exec, then_6:Exec)
   n2 = BreakTransform(InTransform=GetDeltaTransformFromRefPose(self=GetOwnerSkeletalMeshComponent(), BoneName=upperarm_r, BaseName=clavicle_r))
   n3 = Branch(execute=@n1.then_3, Condition="float >= float"(A=BreakRotator(InRot=@n2.Rotation).Yaw, B=100))
   > Set UpperArmRot_l(UpperArmRot_l:bool=1) -> (Output_Get:bool)
   Set UpperArmRot_l(execute=@n3.else, UpperArmRot_l=false)
   Set BlendPoseAlpha(
    execute=@n1.then_4
    BlendPoseAlpha:bool=AND Boolean!(A=Get bIsCrouching(), B=Get bIsTargeting(), C:bool=Get UpperArmRot_l() -> (UpperArmRot_l:bool)) -> (ReturnValue)
   ) -> (Output_Get:bool)
   n4 = GetOwnerSkeletalMeshComponent()
   n5 = BreakRotator(InRot=BreakTransform(InTransform=GetDeltaTransformFromRefPose(self=@n4, BoneName=head, BaseName=spine_05)).Rotation)
   n6 = Branch(execute=@n1.then_1, Condition="float > float"(A=@n5.Yaw))
   > Set Look up Alpha(Look up Alpha:double=0.4) -> (Output_Get:double)
   Set Look up Alpha(execute=@n6.else, Look up Alpha=0)
   n7 = Branch(execute=@n1.then_2, Condition="float < float"(A=@n5.Yaw, B=-15))
   > Set AimUp(AimUp:double=1) -> (Output_Get:double)
   Set AimUp(execute=@n7.else, AimUp=0)
   n8 = Gravity Override State Function DEPRECATED(
    execute=@n1.then_0
    CharacterPartAnimInstance=Self-Reference()
    GameTimeInSeconds=GetGameTimeInSeconds()
    GravityOverrideParams=Get Gravity Override Params()
   )
   > Set GravityOverride(GravityOverride=@n8)
  ]}
]
variables [
 AimUp {type=number}
 BlendPoseAlpha {type=boolean}
 crouchAlpha {type=number, default=1}
 FixCollarClip {type=boolean}
 Gravity Override Params {type=GravityOverrideParamsStruct, default {skydive_Parachuting {windAmplitude=<100 -500 400>}}}
 GravityOverride {type=Vector}
 Look up Alpha {type=number}
 TorsoRotateR {type=boolean}
 UpperArmRot_l {type=boolean}
]
